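# Physics asset for the MountainBase heavy bag prop, letting it swing as a simulated bone chain.
# Zero-length capsules act as spheres on Anim and seven dyn_bag bones, small on the upper links and much larger on the last three to give the bag volume.
# Constraints link each bone to the previous one back to Anim, with twist locked and a much tighter swing limit on the final joint.
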PhysicsAsset /BRCosmetics/Environments/Apollo/Sets/MountainBase/Props/Meshes/HeavyBag/Meshes/Base_HeavyBag_02_Physics
bodies {
 Anim [{shapeType=Capsule, length=0, radius=3}]
 dyn_bag_01 [{shapeType=Capsule, length=0, radius=5, center=<0.1 0 0>}]
 dyn_bag_02 [{shapeType=Capsule, length=0, radius=5, center=<0.1 0 0>}]
 dyn_bag_03 [{shapeType=Capsule, length=0, radius=5, center=<0.1 0 0>}]
 dyn_bag_04 [{shapeType=Capsule, length=0, radius=5, center=<0.1 0 0>}]
 dyn_bag_05 [{shapeType=Capsule, length=0, radius=30, center=<0.1 0 0>}]
 dyn_bag_06 [{shapeType=Capsule, length=0, radius=30, center=<0.5 0 0>}]
 dyn_bag_07 [{shapeType=Capsule, length=0, radius=30, center=<1 0 0>}]
}
constraints [
 {bone1Name=dyn_bag_01, bone2Name=Anim, swing1LimitDegrees=45, swing2LimitDegrees=45, twistLimitDegrees=45, twistMotion=Locked}
 {bone1Name=dyn_bag_02, bone2Name=dyn_bag_01, swing1LimitDegrees=45, swing2LimitDegrees=45, twistLimitDegrees=45, twistMotion=Locked}
 {bone1Name=dyn_bag_03, bone2Name=dyn_bag_02, swing1LimitDegrees=45, swing2LimitDegrees=45, twistLimitDegrees=45, twistMotion=Locked}
 {bone1Name=dyn_bag_04, bone2Name=dyn_bag_03, swing1LimitDegrees=45, swing2LimitDegrees=45, twistLimitDegrees=45, twistMotion=Locked}
 {bone1Name=dyn_bag_05, bone2Name=dyn_bag_04, swing1LimitDegrees=45, swing2LimitDegrees=45, twistLimitDegrees=45, twistMotion=Locked}
 {bone1Name=dyn_bag_06, bone2Name=dyn_bag_05, swing1LimitDegrees=45, swing2LimitDegrees=45, twistLimitDegrees=45, twistMotion=Locked}
 {bone1Name=dyn_bag_07, bone2Name=dyn_bag_06, swing1LimitDegrees=5, swing2LimitDegrees=5, twistLimitDegrees=45, twistMotion=Locked}
]
